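# Additive, unlit, two-sided mesh-particle material for the Ashton Milo backpack explosion effect.
# Cloud-distorted panning cracks and turbulence combine with Fresnel, tinted between outer and inner colors for emissive.
# Opacity reuses that pattern with a sphere mask, depth fades and camera-facing falloff.
# Particle color channels drive brightness, falloff, mask and alpha, while dynamic parameters control noise, Fresnel exponent and switchable smoke-noise refraction.
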Material /Game/Athena/Items/Gameplay/BackPacks/Ashton/Milo/FX/Secondary/M_Assault_Ashton_Milo_Explosion
graph {
 properties {
  bUsedWithMeshParticles=true
  bUsedWithNiagaraMeshParticles=true
  blendMode=BLEND_Additive
  pixelDepthOffsetMode=PDOM_Legacy
  refractionMethod=RM_IndexOfRefraction
  shadingModel=MSM_Unlit
  twoSided=true
  usedShadingModels=Unlit
 }
 nodes [
  n1 = TextureCoordinate() {uTiling=2}
  n2 = Panner(Coordinate=TextureCoordinate()) {speedX=0.2, speedY=0.3}
  n3 = Multiply(A=ComponentMask(None=TextureSample(UVs=@n2) {texture=/Game/Effects/Fort_Effects/PlayerHealthDamage/T_DBNO_Cloud}.RGB) {g=true, r=true}) {constB=0.08}
  n4 = TextureSample(UVs=Panner(Coordinate=Add(A=@n1, B=@n3)) {speedX=0.5}) {texture=/Game/Effects/Fort_Effects/Textures/T_Cracks_Gradient}
  n5 = TextureCoordinate()
  n6 = TextureSample(UVs=Panner(Coordinate=Add(A=@n5, B=@n3)) {speedX=-0.3}) {texture=/Game/Global/Textures/Default/Caustics/T_Turbulence_Seamless}
  n7 = Multiply(A=Multiply(A=@n4.R) {constB=3}, B=@n6.R)
  n8 = Fresnel(ExponentIn=Constant() {r=1.75}, BaseReflectFractionIn=Constant())
  n9 = Add(
   A=Multiply(A=MaterialFunctionCall!(Blend=@n7, Base=@n6.R) {materialFunction=/Engine/Functions/Engine_MaterialFunctions03/Blends/Blend_Lighten}.Result) {constB=2.2}
   B=@n8
  )
  n10 = VectorParameter() {parameterName=color, defaultValue=<rgba 0.0162 0.2474 0.825 1>}
  n11 = DynamicParameter() {paramNames [NoiseLerp, NoiseHardness, FresnelExp, RefractionNoise]}
  n12 = MaterialFunctionCall(Brightness=Constant() {r=0.5}, Exponent=@n11.FresnelExp) {materialFunction=&MF_Fresnel_FX}
  n13 = Multiply(A=@n12.Fresnel, B=@n12.Fresnel)
  n14 = TextureCoordinate() {uTiling=3, vTiling=3}
  n15 = LinearInterpolate(
   A=TextureSample(UVs=Panner(Coordinate=@n14) {speedX=0.2, speedY=0.3}) {texture=&T_NoisePack_01}.R
   B=Constant() {r=1}
   Alpha=@n11.NoiseLerp
  )
  n16 = Power(Base=@n6.R) {constExponent=2.1}
  n17 = MaterialFunctionCall(Blend=Multiply(A=@n7) {constB=1.8}, Base=LinearInterpolate(A=@n4.R, B=Multiply(A=@n16) {constB=4})) {materialFunction=&Blend_Lighten}
  n18 = Constant() {r=1.2}
  n19 = Constant()
  n20 = Add(
   A=Multiply(A=Multiply(A=@n13) {constB=3}, B=Multiply(A=@n15, B=@n11.NoiseHardness))
   B=Multiply(A=Multiply(A=@n17.Result) {constB=2}, B=Fresnel(ExponentIn=@n18, BaseReflectFractionIn=@n19))
  )
  n21 = LinearInterpolate(
   A=Desaturation(None=Multiply(A=@n9, B=@n10.RGB) {constB=35}, Fraction=Constant() {r=-1.8})
   B=VectorParameter() {parameterName=Color Inner, defaultValue=<rgba 0.4367 1.8504 14 0>}.RGB
   Alpha=@n20
  )
  n22 = MaterialFunctionCall(LDR Input=Clamp(None=@n21), HDR Tint=@n21) {materialFunction=&DeriveHDRfromLDR}
  n23 = ParticleColor()
  MaterialFunctionCall(Emissive=Multiply(A=StaticSwitchParameter(True=@n21, False=@n22.RemappedRange) {parameterName=LDR}, B=@n23.R)) {materialFunction=&MF_FX_EmissiveAdjustments}
  n24 = Multiply(A=@n8) {constB=2.1}
  n25 = MaterialFunctionCall(Float2=TextureCoordinate()) {materialFunction=&BreakOutFloat2Components}
  n26 = Constant()
  n27 = ParticleColor()
  n28 = OneMinus(None=DepthFade(FadeDistance=ScalarParameter() {defaultValue=250, parameterName=Param_2}) {fadeDistanceDefault=200})
  n29 = Add(
   A=Multiply(A=Add(A=@n20, B=@n24), B=SphereMask(A=@n25.G, B=@n26, Radius=@n27.B, Hardness=@n27.B) {attenuationRadius=0.5})
   B=StaticSwitchParameter(True=Constant(), False=Power(Base=@n28) {constExponent=3}) {parameterName=Depth Fade}
  )
  n30 = ShadingPathSwitch(Default=DepthFade(Opacity=@n29) {fadeDistanceDefault=160}, Mobile=@n29) {desc=Migrated from FeatureLevelSwitch}
  n31 = TwoSidedSign()
  n32 = DotProduct(A=CameraVectorWS(), B=VertexNormalWS())
  n33 = ParticleColor()
  Multiply(A=@n23.A, B=Multiply(A=LinearInterpolate(A=@n30, B=@n30, Alpha=@n31), B=Power(Base=@n32, Exp=@n33.G) {constExponent=3.4}))
  n34 = TextureCoordinate()
  Reroute(None=TextureSample(UVs=Panner(Coordinate=@n34) {speedY=1}) {samplerType=SAMPLERTYPE_Normal, texture=&T_Weap_Distort_N_02}.RGB)
  n35 = ScalarParameter() {defaultValue=3, parameterName=U Tile2}
  n36 = ScalarParameter() {defaultValue=2, parameterName=V Tile2}
  n37 = TextureCoordinate() {uTiling=2}
  n38 = ComponentMask(None=TextureSample(UVs=Panner(Coordinate=@n37) {speedX=0.2, speedY=-0.1}) {samplerType=SAMPLERTYPE_Normal, texture=/Game/Global/Textures/Default/Noise/T_SwirlNoise_N}.RGB) {g=true, r=true}
  n39 = Add(A=Multiply(A=TextureCoordinate(), B=AppendVector(A=@n35, B=@n36)), B=Multiply(A=Constant2Vector() {r=-1}, B=Multiply(A=@n38) {constB=0.11}))
  n40 = Multiply(
   A=@n11.RefractionNoise
   B=TextureSample(UVs=Panner(Coordinate=@n39) {bFractionalPart=true, speedX=0.45, speedY=1.18}) {texture=&T_Inky_Smoke_Tile}.R
  )
  StaticSwitchParameter(True=Constant() {r=1}, False=LinearInterpolate(Alpha=Multiply(A=@n23.A, B=@n40)) {constA=1, constB=1.03}) {parameterName=Refraction}
 ]
}
