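# Translucent unlit trail effect material for the Ashton Milo backpack, flagged for ribbons, beam trails, mesh particles and skeletal meshes.
# Normal-map-distorted panning smoke, circuitboard patterns and sphere-mask falloffs form the trail mask, with a dynamic parameter shaping the gradient.
# That mask blends HDR blues for emissive, with a flat-color mobile fallback.
# ToMatchPoison blends toward an alternate look: curve-atlas-colored, particle-tinted circuitry and a noise-line mask with optional flicker.
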Material /Game/Athena/Items/Gameplay/BackPacks/Ashton/Milo/FX/Secondary/M_Assault_Ashton_Milo_Trail
graph {
 properties {
  bUsedWithBeamTrails=true
  bUsedWithMeshDeformer=true
  bUsedWithMeshParticles=true
  bUsedWithNiagaraMeshParticles=true
  bUsedWithNiagaraRibbons=true
  bUsedWithSkeletalMesh=true
  blendMode=BLEND_Translucent
  pixelDepthOffsetMode=PDOM_Legacy
  shadingModel=MSM_Unlit
  usedShadingModels=Unlit
 }
 nodes [
  Constant() {r=1}
  n1 = TextureCoordinate() {uTiling=0.5}
  n2 = TextureSample() {samplerType=SAMPLERTYPE_Normal, texture=&T_Weap_Distort_N_02}
  n3 = Constant() {r=0.75}
  n4 = Multiply(A=ComponentMask(None=MaterialFunctionCall(Normal=@n2.RGB, Flatness=@n3) {materialFunction=&FlattenNormal}.Result) {g=true, r=true}) {constB=0.18}
  n5 = TextureSample(UVs=Panner(Coordinate=Add(A=@n1, B=@n4)) {speedX=-2}) {texture=/Game/Effects/Fort_Effects/Textures/T_InkySmokeTrail}
  n6 = TextureSample() {samplerType=SAMPLERTYPE_Normal, texture=&T_Weap_Distort_N_02}
  n7 = Add(A=TextureCoordinate() {uTiling=1.2}, B=Multiply(A=ComponentMask(None=@n6.RGB) {g=true, r=true}) {constB=0.1})
  n8 = MaterialFunctionCall!(
   Blend=Add(A=@n5.R, B=@n5.R)
   Base=TextureSample(UVs=Panner(Coordinate=@n7) {speedX=-1.8}) {samplerType=SAMPLERTYPE_Grayscale, texture=&T_Smoke_Trail_Soft}.R
  ) {materialFunction=/Engine/Functions/Engine_MaterialFunctions03/Blends/Blend_Lighten}
  n9 = TextureCoordinate()
  n10 = Power(
   Base=SphereMask(
    A=MaterialFunctionCall(Float2=@n9) {materialFunction=&BreakOutFloat2Components}.R
    B=Constant()
    Radius=Constant() {r=0.97}
   ) {hardnessPercent=0}
   Exp=DynamicParameter() {defaultValue=<rgba 2.5 1 1 1>, paramNames [opac Gradient pow, Param2, Param3, Param4]}.opac Gradient pow
  ) {constExponent=2.5}
  n11 = TextureSample(UVs=Panner() {speedX=-1.8}) {texture=/Game/Global/Textures/Default/Masks_and_Gradients/T_Ramp_Virt}
  n12 = TextureCoordinate() {uTiling=2, vTiling=0.82}
  n13 = AppendVector(A=Power(Base=ComponentMask(None=@n12) {r=true}) {constExponent=0.5}, B=ComponentMask(None=@n12) {g=true})
  n14 = Multiply(
   A=@n11.R
   B=TextureSample(UVs=MaterialFunctionCall(XY=@n13) {materialFunction=&Swizzle}.YX) {texture=/Game/Effects/Fort_Effects/Textures/T_FX_Circuitboard}.R
  )
  n15 = TextureCoordinate() {vTiling=0.6}
  n16 = SphereMask(
   A=MaterialFunctionCall(Float2=TextureCoordinate()) {materialFunction=&BreakOutFloat2Components}.G
   B=Constant() {r=0.5}
   Radius=Constant() {r=0.35}
  ) {hardnessPercent=0}
  n17 = Multiply(
   A=TextureSample(UVs=MaterialFunctionCall(XY=@n15) {materialFunction=&Swizzle}.YX) {texture=&T_FX_Circuitboard}.G
   B=Multiply(A=@n16, B=@n16)
  )
  n18 = TextureSample(UVs=Panner() {speedX=-1}) {texture=&T_NoisePack_01}
  n19 = SphereMask(
   A=MaterialFunctionCall(Float2=TextureCoordinate()) {materialFunction=&BreakOutFloat2Components}.R
   B=Constant() {r=1}
   Radius=Constant() {r=1}
  ) {hardnessPercent=0}
  n20 = Multiply(A=Subtract(A=Add(A=@n14, B=@n17), B=Multiply(A=@n18.G, B=@n19))) {constB=3.1}
  n21 = Add(A=Multiply(A=Multiply(A=@n8.Result, B=@n8.Result), B=@n10), B=Clamp(None=Multiply(A=@n20, B=@n10)))
  n22 = SphereMask(
   A=MaterialFunctionCall(Float2=TextureCoordinate()) {materialFunction=&BreakOutFloat2Components}.G
   B=Constant() {r=0.5}
   Radius=Constant() {r=0.5}
  ) {hardnessPercent=0}
  n23 = SphereMask(A=TextureCoordinate(), B=Constant2Vector() {g=0.5, r=1}, Hardness=Constant()) {attenuationRadius=1}
  n24 = Multiply(A=Multiply(A=Multiply(A=@n22, B=@n22), B=Power(Base=@n23) {constExponent=1.5})) {constB=2}
  n25 = WorldPosition()
  n26 = LinearInterpolate(
   A=ComponentMask(None=TransformPosition(None=@n25.XYZ) {transformSourceType=TRANSFORMPOSSOURCE_World}) {b=true}
   B=ComponentMask(None=TextureCoordinate()) {r=true}
   Alpha=ScalarParameter() {parameterName=MaskPoison}
  )
  n27 = ScalarParameter() {defaultValue=128, parameterName=LineScale2}
  n28 = VectorParameter() {parameterName=LineSpeed, defaultValue=<rgba 0.1 -0.1 0 0>}
  n29 = Divide(A=@n28.G) {constB=10000000}
  n30 = TextureSample(UVs=Panner(Coordinate=Divide(A=@n26, B=@n27) {constB=32}, Speed=AppendVector(A=@n28.G, B=@n29))) {samplerType=SAMPLERTYPE_Normal, texture=/Game/Global/Textures/Default/Noise/64x64Noise}
  n31 = ScalarParameter() {defaultValue=0.9, parameterName=LineSub}
  n32 = Saturate(None=Multiply(A=Subtract(A=@n30.G, B=@n31) {constB=0.99}) {constB=15})
  n33 = Constant() {r=1}
  n34 = ComponentMask(None=TransformPosition(None=WorldPosition().XYZ) {transformSourceType=TRANSFORMPOSSOURCE_World}) {b=true}
  n35 = ComponentMask(None=TextureCoordinate()) {g=true}
  n36 = ScalarParameter() {parameterName=MaskPoison}
  n37 = AppendVector(
   A=ComponentMask(None=TextureCoordinate() {vTiling=0.5}) {r=true}
   B=Divide(A=LinearInterpolate(A=@n34, B=@n35, Alpha=@n36), B=ScalarParameter() {defaultValue=128, parameterName=FlickerScale}) {constB=32}
  )
  n38 = Ceil(None=Multiply(A=Time()) {constB=3})
  n39 = VectorParameter() {parameterName=Param_1, defaultValue=<rgba 0.1 -0.1 0 1>}
  n40 = AppendVector(A=@n39.R, B=Divide(A=@n39.R) {constB=1000})
  n41 = VectorParameter() {defaultValue=<rgba 0.3 -0.1 0 1>}
  n42 = AppendVector(A=Divide(A=@n41.G) {constB=1000}, B=@n41.G)
  n43 = Subtract(
   A=TextureSample(UVs=Panner(Coordinate=@n37, Time=@n38, Speed=@n40)) {samplerType=SAMPLERTYPE_LinearColor, texture=/Game/UI/Foundation/Textures/Pattern/T_CubePattern}.R
   B=TextureSample(UVs=Panner(Coordinate=@n37, Time=@n38, Speed=@n42)) {samplerType=SAMPLERTYPE_LinearColor, texture=&T_CubePattern}.G
  )
  n44 = Clamp(None=OneMinus(None=Round(None=@n43)))
  n45 = ScalarParameter() {parameterName=Flicker}
  n46 = ScalarParameter() {parameterName=ToMatchPoison}
  LinearInterpolate(
   A=Multiply(A=Multiply(A=@n21, B=@n24)) {constB=2}
   B=Multiply(A=LinearInterpolate(Alpha=@n32) {constA=0.5, constB=2}, B=LinearInterpolate(A=@n33, B=@n44, Alpha=@n45))
   Alpha=@n46
  )
  n47 = Divide(A=@n26, B=ScalarParameter() {defaultValue=12, parameterName=LineScale1}) {constB=32}
  n48 = AppendVector(A=@n28.R, B=Divide(A=@n28.R) {constB=1000})
  LinearInterpolate(Alpha=TextureSample(UVs=Panner(Coordinate=@n47, Speed=@n48)) {samplerType=SAMPLERTYPE_Normal, texture=&64x64Noise}.R) {constA=1, constAlpha=0.1, constB=0.55}
  n49 = Constant3Vector() {constant=<rgba 0.1872 0.5822 1 0>}
  n50 = LinearInterpolate(
   A=Multiply(A=Multiply(A=@n11.R, B=@n49) {constB=20}, B=Constant3Vector() {constant=<rgba 0.5204 1.5244 12 0>})
   B=Constant3Vector() {constant=<rgba 0.1919 12.6548 20 0>}
   Alpha=@n21
  )
  n51 = MaterialFunctionCall(LDR Input=Clamp(None=@n50), HDR Tint=@n50) {materialFunction=&DeriveHDRfromLDR}
  n52 = Constant3Vector() {constant=<rgba 0.0288 1.8982 3 0>}
  n53 = TextureCoordinate() {coordinateIndex=1, uTiling=1.5}
  n54 = ComponentMask(None=@n53) {r=true}
  n55 = Power(Base=ComponentMask(None=@n53) {g=true}, Exp=ScalarParameter() {defaultValue=1, parameterName=Tiling_b})
  n56 = TextureSample(UVs=Panner(Coordinate=AppendVector(A=@n54, B=@n55)) {speedX=1}) {texture=/Game/Effects/Fort_Effects/Textures/T_FX_Circuitboard2}
  n57 = TextureCoordinate() {coordinateIndex=1, uTiling=1.5}
  n58 = ComponentMask(None=@n57) {r=true}
  n59 = Power(Base=ComponentMask(None=@n57) {g=true}, Exp=ScalarParameter() {defaultValue=1, parameterName=Tiling_c})
  n60 = TextureSample(UVs=Panner(Coordinate=AppendVector(A=@n58, B=@n59)) {speedX=1.5}) {texture=&T_FX_Circuitboard2}
  n61 = Constant() {r=0.3}
  n62 = TextureCoordinate() {coordinateIndex=1, uTiling=1.5}
  n63 = ComponentMask(None=@n62) {g=true}
  n64 = ScalarParameter() {defaultValue=1, parameterName=Tiling_a}
  n65 = Panner(Coordinate=AppendVector(A=ComponentMask(None=@n62) {r=true}, B=Power(Base=@n63, Exp=@n64))) {speedX=2}
  n66 = CurveAtlasRowParameter(CurveTime=LinearInterpolate(
    A=LinearInterpolate(A=@n56.G, B=@n60.R, Alpha=@n61)
    B=TextureSample(UVs=@n65) {texture=&T_FX_Circuitboard2}.B
    Alpha=@n61
   )) {
   atlas=&AthenaCurveAtlas01
   curve=/Game/Effects/Fort_Effects/Creatures/Husk/Zapper/Zapper_Default
   defaultValue=12
   group=Color Curve
   parameterName=ColorCurve
  }
  n67 = MaterialFunctionCall(LDR Input=@n66, HDR Tint=@n66) {materialFunction=&DeriveHDRfromLDR}
  n68 = ParticleColor()
  MaterialFunctionCall(Emissive=LinearInterpolate(
    A=ShadingPathSwitch(Default=@n51.RemappedRange, Mobile=@n52) {desc=Migrated from FeatureLevelSwitch}
    B=Multiply(A=@n67.RemappedRange, B=@n68.RGB)
    Alpha=@n46
   )) {materialFunction=&MF_FX_EmissiveAdjustments}
 ]
}
